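# Conversation database for Season 36 Big Dill character story quests, covering the Tasty Bois breakup and comeback show invites.
# Four branches share one entry tag, each gated on an active quest and the NPC's conversation tag: WartyBrine (quest 02), Brat (03), DJBop and DJYonder (both 04).
# Each runs three linear speech lines, ending on a choice tag that marks the story step; both DJ branches emit the same tag.
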ConversationDatabase /BattlePassS36_Quests/Items/Conversations/CharacterStories/Conversation_Quest_S36_CharacterStories_BigDill_01
entry_points [{
  tag=Conversation.Entry.Season36.BigDill.CharacterStories.01
  destinations [#1, #2, #3, #4]
 }]
nodes [
 ConversationEntryPointNode_2 {
  id=#4
  class=ConversationEntryPointNode
  outputs [#5]
  properties {entryTag {tagName=Conversation.Entry.Season36.BigDill.CharacterStories.01}}
 }
 Speech_C_8 {
  id=#5
  class=Speech_C
  outputs [#6]
  properties {
   softTaskIcon=&T_UI_NPC_Conversation_Give
   generalConfig {defaultMessage="Is this a ticket to Big Dill's big comeback show? Don't mind if I do!"}
  }
  sub_nodes [
   ConversationChoiceNode_6 {class=ConversationChoiceNode, properties {defaultChoiceDisplayText=Deliver Invite...}}
   FortConversationRequirement_ControllerRequirement_4 {
    class=FortConversationRequirement_ControllerRequirement
    properties {
     nodeName=Require: Is Quest Active
     requirement {
      class=FortControllerRequirement_IsQuestActive
      properties {questDefinition=/BattlePassS36_Quests/Items/QuestItems/CharacterStories/Quest_S36_CharacterStories_BigDill_04}
     }
    }
   }
   FortConversationRequirement_ControllerRequirement_5 {
    class=FortConversationRequirement_ControllerRequirement
    properties {
     nodeName=Require: NPC Has Conversation Tag
     participantId {tagName=Conversation.Participant.NPC}
     requirement {class=FortControllerRequirement_NPCHasConversationEntryTag, properties {queryTag {tagName=Conversation.Entry.DJBop}}}
    }
   }
  ]
 }
 Speech_C_9 {
  id=#6
  class=Speech_C
  outputs [#7]
  properties {generalConfig {defaultMessage="This will be a DISASTER. "}}
  sub_nodes [ConversationChoiceNode_7 {class=ConversationChoiceNode, properties {defaultChoiceDisplayText=Continue}}]
 }
 Speech_C_10 {
  id=#7
  class=Speech_C
  outputs []
  properties {generalConfig {defaultMessage="And I CAN'T WAIT!!"}}
  sub_nodes [ConversationChoiceNode_8 {
    class=ConversationChoiceNode
    properties {defaultChoiceDisplayText=Continue, choiceTags {gameplayTags [{tagName=Conversation.Choice.BigDill.CharacterStories.03}]}}
   }]
 }
 ConversationEntryPointNode_1 {
  id=#3
  class=ConversationEntryPointNode
  outputs [#8]
  properties {entryTag {tagName=Conversation.Entry.Season36.BigDill.CharacterStories.01}}
 }
 Speech_C_5 {
  id=#8
  class=Speech_C
  outputs [#9]
  properties {
   softTaskIcon=&T_UI_NPC_Conversation_Give
   generalConfig {defaultMessage=Is this an invite to a Big Dill 'concert'? Ugh.}
  }
  sub_nodes [
   ConversationChoiceNode_9 {class=ConversationChoiceNode, properties {defaultChoiceDisplayText=Deliver Invite...}}
   FortConversationRequirement_ControllerRequirement_6 {
    class=FortConversationRequirement_ControllerRequirement
    properties {
     nodeName=Require: Is Quest Active
     requirement {
      class=FortControllerRequirement_IsQuestActive
      properties {questDefinition=&Quest_S36_CharacterStories_BigDill_04}
     }
    }
   }
   FortConversationRequirement_ControllerRequirement_7 {
    class=FortConversationRequirement_ControllerRequirement
    properties {
     nodeName=Require: NPC Has Conversation Tag
     participantId {tagName=Conversation.Participant.NPC}
     requirement {class=FortControllerRequirement_NPCHasConversationEntryTag, properties {queryTag {tagName=Conversation.Entry.DJYonder}}}
    }
   }
  ]
 }
 Speech_C_6 {
  id=#9
  class=Speech_C
  outputs [#10]
  properties {generalConfig {defaultMessage="(reading) \"Stay salty. The first four rows get brined.\""}}
  sub_nodes [ConversationChoiceNode_10 {class=ConversationChoiceNode, properties {defaultChoiceDisplayText=Continue}}]
 }
 Speech_C_7 {
  id=#10
  class=Speech_C
  outputs []
  properties {generalConfig {defaultMessage=I... think I'll pass.}}
  sub_nodes [ConversationChoiceNode_11 {
    class=ConversationChoiceNode
    properties {defaultChoiceDisplayText=Continue, choiceTags {gameplayTags [{tagName=Conversation.Choice.BigDill.CharacterStories.03}]}}
   }]
 }
 ConversationEntryPointNode_0 {
  id=#2
  class=ConversationEntryPointNode
  outputs [#11]
  properties {entryTag {tagName=Conversation.Entry.Season36.BigDill.CharacterStories.01}}
 }
 Speech_C_1 {
  id=#11
  class=Speech_C
  outputs [#12]
  properties {
   softTaskIcon=&T_UI_NPC_MainQuest
   generalConfig {defaultMessage=Big Dill is breaking up the Bois? That's a real tough sausage to swallow.}
  }
  sub_nodes [
   ConversationChoiceNode_3 {class=ConversationChoiceNode, properties {defaultChoiceDisplayText="Tasty Bois... "}}
   FortConversationRequirement_ControllerRequirement_2 {
    class=FortConversationRequirement_ControllerRequirement
    properties {
     nodeName=Require: Is Quest Active
     requirement {
      class=FortControllerRequirement_IsQuestActive
      properties {questDefinition=/BattlePassS36_Quests/Items/QuestItems/CharacterStories/Quest_S36_CharacterStories_BigDill_03}
     }
    }
   }
   FortConversationRequirement_ControllerRequirement_3 {
    class=FortConversationRequirement_ControllerRequirement
    properties {
     nodeName=Require: NPC Has Conversation Tag
     participantId {tagName=Conversation.Participant.NPC}
     requirement {class=FortControllerRequirement_NPCHasConversationEntryTag, properties {queryTag {tagName=Conversation.Entry.Brat}}}
    }
   }
  ]
 }
 Speech_C_2 {
  id=#12
  class=Speech_C
  outputs [#13]
  properties {generalConfig {defaultMessage="It's a sweet'n'sour feeling. Guess I'll go back to selling out stadiums solo. "}}
  sub_nodes [ConversationChoiceNode_4 {class=ConversationChoiceNode, properties {defaultChoiceDisplayText=Continue}}]
 }
 Speech_C_4 {
  id=#13
  class=Speech_C
  outputs []
  properties {generalConfig {defaultMessage="Too bad, I was liking the more intimate venues."}}
  sub_nodes [ConversationChoiceNode_5 {
    class=ConversationChoiceNode
    properties {defaultChoiceDisplayText=Continue, choiceTags {gameplayTags [{tagName=Conversation.Choice.BigDill.CharacterStories.02}]}}
   }]
 }
 ConversationEntryPointNode_3 {
  id=#1
  class=ConversationEntryPointNode
  outputs [#14]
  properties {entryTag {tagName=Conversation.Entry.Season36.BigDill.CharacterStories.01}}
 }
 Speech_C_3 {
  id=#14
  class=Speech_C
  outputs [#15]
  properties {
   softTaskIcon=&T_UI_NPC_MainQuest
   generalConfig {defaultMessage="I've had some real tasty times with my boi The Brat, but this star's too bright to shine in a pair."}
  }
  sub_nodes [
   ConversationChoiceNode_0 {class=ConversationChoiceNode, properties {defaultChoiceDisplayText="Tasty Bois... "}}
   FortConversationRequirement_ControllerRequirement_0 {
    class=FortConversationRequirement_ControllerRequirement
    properties {
     nodeName=Require: Is Quest Active
     requirement {
      class=FortControllerRequirement_IsQuestActive
      properties {questDefinition=/BattlePassS36_Quests/Items/QuestItems/CharacterStories/Quest_S36_CharacterStories_BigDill_02}
     }
    }
   }
   FortConversationRequirement_ControllerRequirement_1 {
    class=FortConversationRequirement_ControllerRequirement
    properties {
     nodeName=Require: NPC Has Conversation Tag
     participantId {tagName=Conversation.Participant.NPC}
     requirement {class=FortControllerRequirement_NPCHasConversationEntryTag, properties {queryTag {tagName=Conversation.Entry.WartyBrine}}}
    }
   }
  ]
 }
 Speech_C_28 {
  id=#15
  class=Speech_C
  outputs [#16]
  properties {generalConfig {defaultMessage=The Tasty Bois are done. It's time for Big Dill to reclaim his spotlight... Solo.}}
  sub_nodes [ConversationChoiceNode_1 {class=ConversationChoiceNode, properties {defaultChoiceDisplayText=Continue}}]
 }
 Speech_C_0 {
  id=#16
  class=Speech_C
  outputs []
  properties {generalConfig {defaultMessage=Would you tell The Brat for me? I don't wanna deal with any drama.}}
  sub_nodes [ConversationChoiceNode_2 {
    class=ConversationChoiceNode
    properties {defaultChoiceDisplayText=Continue, choiceTags {gameplayTags [{tagName=Conversation.Choice.BigDill.CharacterStories.01}]}}
   }]
 }
]
