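# Conversation database for DJ Yonder's Season 36 character story quests, in which Yonder asks the player to message Big Dill and the player later delivers a message.
# Two entry nodes share one tag and lead to three-speech linear branches.
# Active-quest and NPC-tag requirements gate each branch: quest 02 at DJYonder, quest 04 at WartyBrine.
# Each final Continue choice carries its own CharacterStories choice tag, 01 or 02.
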ConversationDatabase /BattlePassS36_Quests/Items/Conversations/CharacterStories/Conversation_Quest_S36_CharacterStories_DJYonder_01
entry_points [{
  tag=Conversation.Entry.Season36.DJYonder.CharacterStories.01
  destinations [#1, #2]
 }]
nodes [
 ConversationEntryPointNode_0 {
  id=#2
  class=ConversationEntryPointNode
  outputs [#3]
  properties {entryTag {tagName=Conversation.Entry.Season36.DJYonder.CharacterStories.01}}
 }
 Speech_C_1 {
  id=#3
  class=Speech_C
  outputs [#4]
  properties {
   softTaskIcon=&T_UI_NPC_MainQuest
   generalConfig {defaultMessage="Ohhh, DJ Yonder's BIG mad. Like, for reals. Oops."}
  }
  sub_nodes [
   ConversationChoiceNode_5 {class=ConversationChoiceNode, properties {defaultChoiceDisplayText=Deliver Message...}}
   FortConversationRequirement_ControllerRequirement_4 {
    class=FortConversationRequirement_ControllerRequirement
    properties {
     nodeName=Require: Is Quest Active
     requirement {
      class=FortControllerRequirement_IsQuestActive
      properties {questDefinition=/BattlePassS36_Quests/Items/QuestItems/CharacterStories/Quest_S36_CharacterStories_DJYonder_04}
     }
    }
   }
   FortConversationRequirement_ControllerRequirement_5 {
    class=FortConversationRequirement_ControllerRequirement
    properties {
     nodeName=Require: NPC Has Conversation Tag
     participantId {tagName=Conversation.Participant.NPC}
     requirement {class=FortControllerRequirement_NPCHasConversationEntryTag, properties {queryTag {tagName=Conversation.Entry.WartyBrine}}}
    }
   }
  ]
 }
 Speech_C_2 {
  id=#4
  class=Speech_C
  outputs [#5]
  properties {generalConfig {defaultMessage="Maybe I shouldn't have hired someone to hack his phone and steal his beats, huh?"}}
  sub_nodes [ConversationChoiceNode_6 {class=ConversationChoiceNode, properties {defaultChoiceDisplayText=Continue}}]
 }
 Speech_C_4 {
  id=#5
  class=Speech_C
  outputs []
  properties {generalConfig {defaultMessage=I'll talk to Yon and make it right. I'm not looking to become Big Relish.}}
  sub_nodes [ConversationChoiceNode_7 {
    class=ConversationChoiceNode
    properties {defaultChoiceDisplayText=Continue, choiceTags {gameplayTags [{tagName=Conversation.Choice.DJYonder.CharacterStories.02}]}}
   }]
 }
 ConversationEntryPointNode_3 {
  id=#1
  class=ConversationEntryPointNode
  outputs [#6]
  properties {entryTag {tagName=Conversation.Entry.Season36.DJYonder.CharacterStories.01}}
 }
 Speech_C_3 {
  id=#6
  class=Speech_C
  outputs [#7]
  properties {
   softTaskIcon=&T_UI_NPC_MainQuest
   generalConfig {defaultMessage=Big Dill is planning to use his ill-gotten fortune to take the top of the radio charts by force.}
  }
  sub_nodes [
   ConversationChoiceNode_0 {class=ConversationChoiceNode, properties {defaultChoiceDisplayText="Music Competition... "}}
   FortConversationRequirement_ControllerRequirement_0 {
    class=FortConversationRequirement_ControllerRequirement
    properties {
     nodeName=Require: Is Quest Active
     requirement {
      class=FortControllerRequirement_IsQuestActive
      properties {questDefinition=/BattlePassS36_Quests/Items/QuestItems/CharacterStories/Quest_S36_CharacterStories_DJYonder_02}
     }
    }
   }
   FortConversationRequirement_ControllerRequirement_1 {
    class=FortConversationRequirement_ControllerRequirement
    properties {
     nodeName=Require: NPC Has Conversation Tag
     participantId {tagName=Conversation.Participant.NPC}
     requirement {class=FortControllerRequirement_NPCHasConversationEntryTag, properties {queryTag {tagName=Conversation.Entry.DJYonder}}}
    }
   }
  ]
 }
 Speech_C_28 {
  id=#7
  class=Speech_C
  outputs [#8]
  properties {generalConfig {defaultMessage="The poor souls on this Island have suffered enough. They don't need to be deluged by his... \"music\"."}}
  sub_nodes [ConversationChoiceNode_1 {class=ConversationChoiceNode, properties {defaultChoiceDisplayText=Continue}}]
 }
 Speech_C_0 {
  id=#8
  class=Speech_C
  outputs []
  properties {generalConfig {defaultMessage="Send the pickle a message for me. If you do, you'll be a true hero."}}
  sub_nodes [ConversationChoiceNode_2 {
    class=ConversationChoiceNode
    properties {defaultChoiceDisplayText=Continue, choiceTags {gameplayTags [{tagName=Conversation.Choice.DJYonder.CharacterStories.01}]}}
   }]
 }
]
