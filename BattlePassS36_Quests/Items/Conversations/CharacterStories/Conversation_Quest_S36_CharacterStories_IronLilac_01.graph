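# Season 36 character-story quest conversation in which the Iron Lilac NPC discusses Daigo's turn and his Spires, then points the player to Kendo.
# A single entry tag leads into a linear three-line speech chain with no branches.
# The opening choice shows a main-quest icon and requires quest IronLilac_02 to be active and the NPC to carry the IronLilac conversation tag.
# The last line's Continue choice carries a CharacterStories.01 choice tag.
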ConversationDatabase /BattlePassS36_Quests/Items/Conversations/CharacterStories/Conversation_Quest_S36_CharacterStories_IronLilac_01
entry_points [{tag=Conversation.Entry.Season36.IronLilac.CharacterStories.01, destinations [#1]}]
nodes [
 ConversationEntryPointNode_3 {
  id=#1
  class=ConversationEntryPointNode
  outputs [#2]
  properties {entryTag {tagName=Conversation.Entry.Season36.IronLilac.CharacterStories.01}}
 }
 Speech_C_3 {
  id=#2
  class=Speech_C
  outputs [#3]
  properties {
   softTaskIcon=&T_UI_NPC_MainQuest
   generalConfig {defaultMessage="Daigo's changed. He's always been arrogant and reckless, but he wasn't a villain."}
  }
  sub_nodes [
   ConversationChoiceNode_5 {class=ConversationChoiceNode, properties {defaultChoiceDisplayText="Mask Maker Daigo... "}}
   FortConversationRequirement_ControllerRequirement_4 {
    class=FortConversationRequirement_ControllerRequirement
    properties {
     nodeName=Require: Is Quest Active
     requirement {
      class=FortControllerRequirement_IsQuestActive
      properties {questDefinition=/BattlePassS36_Quests/Items/QuestItems/CharacterStories/Quest_S36_CharacterStories_IronLilac_02}
     }
    }
   }
   FortConversationRequirement_ControllerRequirement_5 {
    class=FortConversationRequirement_ControllerRequirement
    properties {
     nodeName=Require: NPC Has Conversation Tag
     participantId {tagName=Conversation.Participant.NPC}
     requirement {class=FortControllerRequirement_NPCHasConversationEntryTag, properties {queryTag {tagName=Conversation.Entry.IronLilac}}}
    }
   }
  ]
 }
 Speech_C_28 {
  id=#3
  class=Speech_C
  outputs [#4]
  properties {generalConfig {defaultMessage="And now... we don't know exactly what these Spires he's summoning are for, but it's NOT good."}}
  sub_nodes [ConversationChoiceNode_6 {class=ConversationChoiceNode, properties {defaultChoiceDisplayText=Continue}}]
 }
 Speech_C_0 {
  id=#4
  class=Speech_C
  outputs []
  properties {generalConfig {defaultMessage="Kendo is helping me look into what's going on with Daigo. If you want to help, he's ready."}}
  sub_nodes [ConversationChoiceNode_7 {
    class=ConversationChoiceNode
    properties {defaultChoiceDisplayText=Continue, choiceTags {gameplayTags [{tagName=Conversation.Choice.IronLilac.CharacterStories.01}]}}
   }]
 }
]
