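# NPC conversation for Season 36 IronLilac character story quest 05, where the player reports to Hope about Daigo possibly being controlled.
# The "Report to Hope..." choice requires that quest to be active and the NPC to carry the IronLilac conversation entry tag.
# Four speech nodes run linearly through Continue choices.
# The last choice carries an IronLilac CharacterStories choice tag numbered 02, unlike the 05 entry.
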ConversationDatabase /BattlePassS36_Quests/Items/Conversations/CharacterStories/Conversation_Quest_S36_CharacterStories_IronLilac_05
entry_points [{tag=Conversation.Entry.Season36.IronLilac.CharacterStories.05, destinations [#1]}]
nodes [
 ConversationEntryPointNode_0 {
  id=#1
  class=ConversationEntryPointNode
  outputs [#2]
  properties {entryTag {tagName=Conversation.Entry.Season36.IronLilac.CharacterStories.05}}
 }
 Speech_C_3 {
  id=#2
  class=Speech_C
  outputs [#3]
  properties {
   softTaskIcon=&T_UI_NPC_MainQuest
   generalConfig {defaultMessage="Kendo thinks Daigo's being controlled? That... makes way more sense, actually."}
  }
  sub_nodes [
   ConversationChoiceNode_5 {class=ConversationChoiceNode, properties {defaultChoiceDisplayText=Report to Hope...}}
   FortConversationRequirement_ControllerRequirement_4 {
    class=FortConversationRequirement_ControllerRequirement
    properties {
     nodeName=Require: Is Quest Active
     requirement {
      class=FortControllerRequirement_IsQuestActive
      properties {questDefinition=/BattlePassS36_Quests/Items/QuestItems/CharacterStories/Quest_S36_CharacterStories_IronLilac_05}
     }
    }
   }
   FortConversationRequirement_ControllerRequirement_5 {
    class=FortConversationRequirement_ControllerRequirement
    properties {
     nodeName=Require: NPC Has Conversation Tag
     participantId {tagName=Conversation.Participant.NPC}
     requirement {class=FortControllerRequirement_NPCHasConversationEntryTag, properties {queryTag {tagName=Conversation.Entry.IronLilac}}}
    }
   }
  ]
 }
 Speech_C_28 {
  id=#3
  class=Speech_C
  outputs [#4]
  properties {generalConfig {defaultMessage="After what The Night Rose experienced... I can't imagine, your mind and body betraying you like that."}}
  sub_nodes [ConversationChoiceNode_6 {class=ConversationChoiceNode, properties {defaultChoiceDisplayText=Continue}}]
 }
 Speech_C_0 {
  id=#4
  class=Speech_C
  outputs [#5]
  properties {generalConfig {defaultMessage="We have to defeat these Spires, save Daigo, find a way into the Spirit Realm, and confront the dark presence there."}}
  sub_nodes [ConversationChoiceNode_0 {class=ConversationChoiceNode, properties {defaultChoiceDisplayText=Continue}}]
 }
 Speech_C_1 {
  id=#5
  class=Speech_C
  outputs []
  properties {generalConfig {defaultMessage="So, uhhh... no biggie. Yeah."}}
  sub_nodes [ConversationChoiceNode_7 {
    class=ConversationChoiceNode
    properties {defaultChoiceDisplayText=Continue, choiceTags {gameplayTags [{tagName=Conversation.Choice.IronLilac.CharacterStories.02}]}}
   }]
 }
]
