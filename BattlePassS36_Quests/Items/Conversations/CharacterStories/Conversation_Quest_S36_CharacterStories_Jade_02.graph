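# Season 36 NPC conversation for Jade's second character-story quest, in which the NPC asks the player to check on the Sprites.
# One entry point leads into three linear Speech nodes, each with its own choice text.
# The opening "Sprites..." choice requires Quest Jade_02 to be active and the NPC to carry the PolishedJade entry tag.
# The final Continue carries a choice tag that, like the entry tag, ends in 01 rather than 02.
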ConversationDatabase /BattlePassS36_Quests/Items/Conversations/CharacterStories/Conversation_Quest_S36_CharacterStories_Jade_02
entry_points [{tag=Conversation.Entry.Season36.PolishedJade.CharacterStories.01, destinations [#1]}]
nodes [
 ConversationEntryPointNode_0 {
  id=#1
  class=ConversationEntryPointNode
  outputs [#2]
  properties {entryTag {tagName=Conversation.Entry.Season36.PolishedJade.CharacterStories.01}}
 }
 Speech_C_3 {
  id=#2
  class=Speech_C
  outputs [#3]
  properties {
   softTaskIcon=&T_UI_NPC_MainQuest
   generalConfig {defaultMessage=The Sprites are what protected me when I was in the Spirit Realm. They're special.}
  }
  sub_nodes [
   ConversationChoiceNode_5 {class=ConversationChoiceNode, properties {defaultChoiceDisplayText=Sprites...}}
   FortConversationRequirement_ControllerRequirement_4 {
    class=FortConversationRequirement_ControllerRequirement
    properties {
     nodeName=Require: Is Quest Active
     requirement {
      class=FortControllerRequirement_IsQuestActive
      properties {questDefinition=/BattlePassS36_Quests/Items/QuestItems/CharacterStories/Quest_S36_CharacterStories_Jade_02}
     }
    }
   }
   FortConversationRequirement_ControllerRequirement_5 {
    class=FortConversationRequirement_ControllerRequirement
    properties {
     nodeName=Require: NPC Has Conversation Tag
     participantId {tagName=Conversation.Participant.NPC}
     requirement {class=FortControllerRequirement_NPCHasConversationEntryTag, properties {queryTag {tagName=Conversation.Entry.PolishedJade}}}
    }
   }
  ]
 }
 Speech_C_28 {
  id=#3
  class=Speech_C
  outputs [#4]
  properties {generalConfig {defaultMessage=They react to the Island and its energies. And some of them have changed a little bit recently.}}
  sub_nodes [ConversationChoiceNode_6 {class=ConversationChoiceNode, properties {defaultChoiceDisplayText=Continue}}]
 }
 Speech_C_0 {
  id=#4
  class=Speech_C
  outputs []
  properties {generalConfig {defaultMessage=Coulda ya play around with them a little and make sure they're okay? For me?}}
  sub_nodes [ConversationChoiceNode_7 {
    class=ConversationChoiceNode
    properties {defaultChoiceDisplayText=Continue, choiceTags {gameplayTags [{tagName=Conversation.Choice.PolishedJade.CharacterStories.01}]}}
   }]
 }
]
