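# Conversation database for the Season 36 Midas character-story quests, giving the AuraCop NPC dialogue about a spy go-between errand.
# One entry point offers two three-line branches, each shown with a main-quest icon and gated on an active quest plus the NPC's AuraCop entry tag.
# Quest 02's branch requests intel from an associate, while quest 04's reacts to it.
# Each branch's final Continue choice carries its own gameplay tag (01 or 02).
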ConversationDatabase /BattlePassS36_Quests/Items/Conversations/CharacterStories/Conversation_Quest_S36_CharacterStories_Midas_01
entry_points [{tag=Conversation.Entry.Season36.AuraCop.CharacterStories, destinations [#1]}]
nodes [
 ConversationEntryPointNode_3 {
  id=#1
  class=ConversationEntryPointNode
  outputs [#2, #3]
  properties {entryTag {tagName=Conversation.Entry.Season36.AuraCop.CharacterStories}}
 }
 Speech_C_1 {
  id=#3
  class=Speech_C
  outputs [#4]
  properties {
   softTaskIcon=&T_UI_NPC_MainQuest
   generalConfig {defaultMessage=Hm. It seems I underestimated Skillet's competence as a spy.}
  }
  sub_nodes [
   ConversationChoiceNode_5 {class=ConversationChoiceNode, properties {defaultChoiceDisplayText="Valuable Intel... "}}
   FortConversationRequirement_ControllerRequirement_4 {
    class=FortConversationRequirement_ControllerRequirement
    properties {
     nodeName=Require: Is Quest Active
     requirement {
      class=FortControllerRequirement_IsQuestActive
      properties {questDefinition=/BattlePassS36_Quests/Items/QuestItems/CharacterStories/Quest_S36_CharacterStories_Midas_04}
     }
    }
   }
   FortConversationRequirement_ControllerRequirement_5 {
    class=FortConversationRequirement_ControllerRequirement
    properties {
     nodeName=Require: NPC Has Conversation Tag
     participantId {tagName=Conversation.Participant.NPC}
     requirement {class=FortControllerRequirement_NPCHasConversationEntryTag, properties {queryTag {tagName=Conversation.Entry.AuraCop}}}
    }
   }
  ]
 }
 Speech_C_2 {
  id=#4
  class=Speech_C
  outputs [#5]
  properties {generalConfig {defaultMessage="Thank you for working as a go-between. There are, as always, eyes on me."}}
  sub_nodes [ConversationChoiceNode_6 {class=ConversationChoiceNode, properties {defaultChoiceDisplayText=Continue}}]
 }
 Speech_C_4 {
  id=#5
  class=Speech_C
  outputs []
  properties {generalConfig {defaultMessage=It seems my new allies don't entirely trust me. Good. They shouldn't.}}
  sub_nodes [ConversationChoiceNode_7 {
    class=ConversationChoiceNode
    properties {defaultChoiceDisplayText=Continue, choiceTags {gameplayTags [{tagName=Conversation.Choice.AuraCop.CharacterStories.02}]}}
   }]
 }
 Speech_C_3 {
  id=#2
  class=Speech_C
  outputs [#6]
  properties {
   softTaskIcon=&T_UI_NPC_MainQuest
   generalConfig {defaultMessage=I could use your help liaising with my... associate.}
  }
  sub_nodes [
   ConversationChoiceNode_0 {class=ConversationChoiceNode, properties {defaultChoiceDisplayText="Field Operative... "}}
   FortConversationRequirement_ControllerRequirement_0 {
    class=FortConversationRequirement_ControllerRequirement
    properties {
     nodeName=Require: Is Quest Active
     requirement {
      class=FortControllerRequirement_IsQuestActive
      properties {questDefinition=/BattlePassS36_Quests/Items/QuestItems/CharacterStories/Quest_S36_CharacterStories_Midas_02}
     }
    }
   }
   FortConversationRequirement_ControllerRequirement_1 {
    class=FortConversationRequirement_ControllerRequirement
    properties {
     nodeName=Require: NPC Has Conversation Tag
     participantId {tagName=Conversation.Participant.NPC}
     requirement {class=FortControllerRequirement_NPCHasConversationEntryTag, properties {queryTag {tagName=Conversation.Entry.AuraCop}}}
    }
   }
  ]
 }
 Speech_C_28 {
  id=#6
  class=Speech_C
  outputs [#7]
  properties {generalConfig {defaultMessage="He's chaotic, but he's effective."}}
  sub_nodes [ConversationChoiceNode_1 {class=ConversationChoiceNode, properties {defaultChoiceDisplayText=Continue}}]
 }
 Speech_C_0 {
  id=#7
  class=Speech_C
  outputs []
  properties {generalConfig {defaultMessage=Track him down and bring me the intel he's collected.}}
  sub_nodes [ConversationChoiceNode_2 {
    class=ConversationChoiceNode
    properties {defaultChoiceDisplayText=Continue, choiceTags {gameplayTags [{tagName=Conversation.Choice.AuraCop.CharacterStories.01}]}}
   }]
 }
]
