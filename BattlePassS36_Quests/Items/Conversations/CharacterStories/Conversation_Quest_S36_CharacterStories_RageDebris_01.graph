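# NPC dialogue database for the Season 36 RageDebris character story, covering the Skillet_02 through Skillet_05 quests.
# One entry tag fans out to seven three-line speech chains, each gated by an active quest and the NPC's conversation entry tag.
# RageDebris speaks in stages 02 and 05, while other NPCs offer alternative chains for 03 and 04.
# Each final Continue choice carries a stage tag (CharacterStories.01–04), shared by that stage's alternatives.
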ConversationDatabase /BattlePassS36_Quests/Items/Conversations/CharacterStories/Conversation_Quest_S36_CharacterStories_RageDebris_01
entry_points [{tag=Conversation.Entry.Season36.RageDebris.CharacterStories, destinations [#1]}]
nodes [
 ConversationEntryPointNode_3 {
  id=#1
  class=ConversationEntryPointNode
  outputs [
   #2
   #3
   #4
   #5
   #6
   #7
   #8
  ]
  properties {entryTag {tagName=Conversation.Entry.Season36.RageDebris.CharacterStories}}
 }
 Speech_C_52 {
  id=#8
  class=Speech_C
  outputs [#9]
  properties {
   softTaskIcon=&T_UI_NPC_MainQuest
   generalConfig {defaultMessage="Awright, so I played possum while you did all the work."}
  }
  sub_nodes [
   ConversationChoiceNode_24 {class=ConversationChoiceNode, properties {defaultChoiceDisplayText="Valuable Intel... "}}
   FortConversationRequirement_ControllerRequirement_14 {
    class=FortConversationRequirement_ControllerRequirement
    properties {
     nodeName=Require: Is Quest Active
     requirement {
      class=FortControllerRequirement_IsQuestActive
      properties {questDefinition=/BattlePassS36_Quests/Items/QuestItems/CharacterStories/Quest_S36_CharacterStories_Skillet_05}
     }
    }
   }
   FortConversationRequirement_ControllerRequirement_15 {
    class=FortConversationRequirement_ControllerRequirement
    properties {
     nodeName=Require: NPC Has Conversation Tag
     participantId {tagName=Conversation.Participant.NPC}
     requirement {class=FortControllerRequirement_NPCHasConversationEntryTag, properties {queryTag {tagName=Conversation.Entry.RageDebris}}}
    }
   }
  ]
 }
 Speech_C_53 {
  id=#9
  class=Speech_C
  outputs [#10]
  properties {generalConfig {defaultMessage="If ye could leave that bit out of your report to the gold man, I'd appreciate it."}}
  sub_nodes [ConversationChoiceNode_23 {class=ConversationChoiceNode, properties {defaultChoiceDisplayText=Continue}}]
 }
 Speech_C_54 {
  id=#10
  class=Speech_C
  outputs []
  properties {generalConfig {defaultMessage="And if ye don't, well... Keep an eye out for hidden 'splosives, maybe."}}
  sub_nodes [ConversationChoiceNode_22 {
    class=ConversationChoiceNode
    properties {defaultChoiceDisplayText=Continue, choiceTags {gameplayTags [{tagName=Conversation.Choice.RageDebris.CharacterStories.04}]}}
   }]
 }
 Speech_C_27 {
  id=#7
  class=Speech_C
  outputs [#11]
  properties {
   softTaskIcon=&T_UI_NPC_MainQuest
   generalConfig {defaultMessage=These Spires are woven from the magics of the Spirit Realm.}
  }
  sub_nodes [
   ConversationChoiceNode_67 {class=ConversationChoiceNode, properties {defaultChoiceDisplayText=Spires...}}
   FortConversationRequirement_ControllerRequirement_44 {
    class=FortConversationRequirement_ControllerRequirement
    properties {
     nodeName=Require: Is Quest Active
     requirement {
      class=FortControllerRequirement_IsQuestActive
      properties {questDefinition=/BattlePassS36_Quests/Items/QuestItems/CharacterStories/Quest_S36_CharacterStories_Skillet_04}
     }
    }
   }
   FortConversationRequirement_ControllerRequirement_45 {
    class=FortConversationRequirement_ControllerRequirement
    properties {
     nodeName=Require: NPC Has Conversation Tag
     participantId {tagName=Conversation.Participant.NPC}
     requirement {class=FortControllerRequirement_NPCHasConversationEntryTag, properties {queryTag {tagName=Conversation.Entry.RoseDepth}}}
    }
   }
  ]
 }
 Speech_C_29 {
  id=#11
  class=Speech_C
  outputs [#12]
  properties {generalConfig {defaultMessage=Their master seeks to gain purchase in this world. We cannot let that happen.}}
  sub_nodes [ConversationChoiceNode_68 {class=ConversationChoiceNode, properties {defaultChoiceDisplayText=Continue}}]
 }
 Speech_C_30 {
  id=#12
  class=Speech_C
  outputs []
  properties {generalConfig {defaultMessage=His thrall... No. I will NEVER have my taken from me again.}}
  sub_nodes [ConversationChoiceNode_69 {
    class=ConversationChoiceNode
    properties {defaultChoiceDisplayText=Continue, choiceTags {gameplayTags [{tagName=Conversation.Choice.RageDebris.CharacterStories.03}]}}
   }]
 }
 Speech_C_24 {
  id=#6
  class=Speech_C
  outputs [#13]
  properties {
   softTaskIcon=&T_UI_NPC_MainQuest
   generalConfig {defaultMessage="These Spires are something new, even to me. And trust me, I've seen a lot."}
  }
  sub_nodes [
   ConversationChoiceNode_34 {class=ConversationChoiceNode, properties {defaultChoiceDisplayText=Spires...}}
   FortConversationRequirement_ControllerRequirement_22 {
    class=FortConversationRequirement_ControllerRequirement
    properties {
     nodeName=Require: Is Quest Active
     requirement {
      class=FortControllerRequirement_IsQuestActive
      properties {questDefinition=&Quest_S36_CharacterStories_Skillet_04}
     }
    }
   }
   FortConversationRequirement_ControllerRequirement_23 {
    class=FortConversationRequirement_ControllerRequirement
    properties {
     nodeName=Require: NPC Has Conversation Tag
     participantId {tagName=Conversation.Participant.NPC}
     requirement {class=FortControllerRequirement_NPCHasConversationEntryTag, properties {queryTag {tagName=Conversation.Entry.UndergroundRebel}}}
    }
   }
  ]
 }
 Speech_C_25 {
  id=#13
  class=Speech_C
  outputs [#14]
  properties {generalConfig {defaultMessage="Whatever this is, I'm betting it's child's play compared to the full power of the Shard."}}
  sub_nodes [ConversationChoiceNode_35 {class=ConversationChoiceNode, properties {defaultChoiceDisplayText=Continue}}]
 }
 Speech_C_26 {
  id=#14
  class=Speech_C
  outputs []
  properties {generalConfig {defaultMessage=I'm following Hope's lead on trusting certain... shadowy allies. For now.}}
  sub_nodes [ConversationChoiceNode_36 {
    class=ConversationChoiceNode
    properties {defaultChoiceDisplayText=Continue, choiceTags {gameplayTags [{tagName=Conversation.Choice.RageDebris.CharacterStories.03}]}}
   }]
 }
 Speech_C_21 {
  id=#5
  class=Speech_C
  outputs [#15]
  properties {
   softTaskIcon=&T_UI_NPC_MainQuest
   generalConfig {defaultMessage="Studies are ongoing on the Spires, and we've had some breakthroughs on how they work."}
  }
  sub_nodes [
   ConversationChoiceNode_27 {class=ConversationChoiceNode, properties {defaultChoiceDisplayText=Spires...}}
   FortConversationRequirement_ControllerRequirement_16 {
    class=FortConversationRequirement_ControllerRequirement
    properties {
     nodeName=Require: Is Quest Active
     requirement {
      class=FortControllerRequirement_IsQuestActive
      properties {questDefinition=&Quest_S36_CharacterStories_Skillet_04}
     }
    }
   }
   FortConversationRequirement_ControllerRequirement_17 {
    class=FortConversationRequirement_ControllerRequirement
    properties {
     nodeName=Require: NPC Has Conversation Tag
     participantId {tagName=Conversation.Participant.NPC}
     requirement {class=FortControllerRequirement_NPCHasConversationEntryTag, properties {queryTag {tagName=Conversation.Entry.IronLilac}}}
    }
   }
  ]
 }
 Speech_C_22 {
  id=#15
  class=Speech_C
  outputs [#16]
  properties {generalConfig {defaultMessage="We've been... selectively redacting some details, on a need-to-know basis."}}
  sub_nodes [ConversationChoiceNode_26 {class=ConversationChoiceNode, properties {defaultChoiceDisplayText=Continue}}]
 }
 Speech_C_23 {
  id=#16
  class=Speech_C
  outputs []
  properties {generalConfig {defaultMessage=Some people with... shadowy histories... might be drawn to misusing our findings.}}
  sub_nodes [ConversationChoiceNode_25 {
    class=ConversationChoiceNode
    properties {defaultChoiceDisplayText=Continue, choiceTags {gameplayTags [{tagName=Conversation.Choice.RageDebris.CharacterStories.03}]}}
   }]
 }
 Speech_C_17 {
  id=#4
  class=Speech_C
  outputs [#17]
  properties {
   softTaskIcon=&T_UI_NPC_MainQuest
   generalConfig {defaultMessage=I've been talking to the Sprites about those Spires that Daigo is summoning.}
  }
  sub_nodes [
   ConversationChoiceNode_31 {class=ConversationChoiceNode, properties {defaultChoiceDisplayText=Daigo...}}
   FortConversationRequirement_ControllerRequirement_20 {
    class=FortConversationRequirement_ControllerRequirement
    properties {
     nodeName=Require: Is Quest Active
     requirement {
      class=FortControllerRequirement_IsQuestActive
      properties {questDefinition=/BattlePassS36_Quests/Items/QuestItems/CharacterStories/Quest_S36_CharacterStories_Skillet_03}
     }
    }
   }
   FortConversationRequirement_ControllerRequirement_21 {
    class=FortConversationRequirement_ControllerRequirement
    properties {
     nodeName=Require: NPC Has Conversation Tag
     participantId {tagName=Conversation.Participant.NPC}
     requirement {class=FortControllerRequirement_NPCHasConversationEntryTag, properties {queryTag {tagName=Conversation.Entry.PolishedJade}}}
    }
   }
  ]
 }
 Speech_C_16 {
  id=#17
  class=Speech_C
  outputs [#18]
  properties {generalConfig {defaultMessage=Those Spires... they're tied to that dark presence in the Spirit Realm.}}
  sub_nodes [ConversationChoiceNode_32 {class=ConversationChoiceNode, properties {defaultChoiceDisplayText=Continue}}]
 }
 Speech_C_15 {
  id=#18
  class=Speech_C
  outputs []
  properties {generalConfig {defaultMessage=The Sprites are afraid of the Spires. And they won't tell me why.}}
  sub_nodes [ConversationChoiceNode_33 {
    class=ConversationChoiceNode
    properties {defaultChoiceDisplayText=Continue, choiceTags {gameplayTags [{tagName=Conversation.Choice.RageDebris.CharacterStories.02}]}}
   }]
 }
 Speech_C_20 {
  id=#3
  class=Speech_C
  outputs [#19]
  properties {
   softTaskIcon=&T_UI_NPC_MainQuest
   generalConfig {defaultMessage="I've been watching Daigo, and I don't think he's in charge. Not at all."}
  }
  sub_nodes [
   ConversationChoiceNode_28 {class=ConversationChoiceNode, properties {defaultChoiceDisplayText=Daigo...}}
   FortConversationRequirement_ControllerRequirement_18 {
    class=FortConversationRequirement_ControllerRequirement
    properties {
     nodeName=Require: Is Quest Active
     requirement {
      class=FortControllerRequirement_IsQuestActive
      properties {questDefinition=&Quest_S36_CharacterStories_Skillet_03}
     }
    }
   }
   FortConversationRequirement_ControllerRequirement_19 {
    class=FortConversationRequirement_ControllerRequirement
    properties {
     nodeName=Require: NPC Has Conversation Tag
     participantId {tagName=Conversation.Participant.NPC}
     requirement {class=FortControllerRequirement_NPCHasConversationEntryTag, properties {queryTag {tagName=Conversation.Entry.CubicVice}}}
    }
   }
  ]
 }
 Speech_C_19 {
  id=#19
  class=Speech_C
  outputs [#20]
  properties {generalConfig {defaultMessage="He's powerful, and dangerous, but he's taking orders from someone... or something."}}
  sub_nodes [ConversationChoiceNode_29 {class=ConversationChoiceNode, properties {defaultChoiceDisplayText=Continue}}]
 }
 Speech_C_18 {
  id=#20
  class=Speech_C
  outputs []
  properties {generalConfig {defaultMessage=And those Spires are the whole point of all of this. But... why?}}
  sub_nodes [ConversationChoiceNode_30 {
    class=ConversationChoiceNode
    properties {defaultChoiceDisplayText=Continue, choiceTags {gameplayTags [{tagName=Conversation.Choice.RageDebris.CharacterStories.02}]}}
   }]
 }
 Speech_C_48 {
  id=#2
  class=Speech_C
  outputs [#21]
  properties {
   softTaskIcon=&T_UI_NPC_MainQuest
   generalConfig {defaultMessage="Yer bringing fresh orders from the gold man, eh?"}
  }
  sub_nodes [
   ConversationChoiceNode_64 {class=ConversationChoiceNode, properties {defaultChoiceDisplayText="Valuable Intel... "}}
   FortConversationRequirement_ControllerRequirement_42 {
    class=FortConversationRequirement_ControllerRequirement
    properties {
     nodeName=Require: Is Quest Active
     requirement {
      class=FortControllerRequirement_IsQuestActive
      properties {questDefinition=/BattlePassS36_Quests/Items/QuestItems/CharacterStories/Quest_S36_CharacterStories_Skillet_02}
     }
    }
   }
   FortConversationRequirement_ControllerRequirement_43 {
    class=FortConversationRequirement_ControllerRequirement
    properties {
     nodeName=Require: NPC Has Conversation Tag
     participantId {tagName=Conversation.Participant.NPC}
     requirement {class=FortControllerRequirement_NPCHasConversationEntryTag, properties {queryTag {tagName=Conversation.Entry.RageDebris}}}
    }
   }
  ]
 }
 Speech_C_47 {
  id=#21
  class=Speech_C
  outputs [#22]
  properties {generalConfig {defaultMessage="Tell ye what, how about we split the work 50/50?"}}
  sub_nodes [ConversationChoiceNode_65 {class=ConversationChoiceNode, properties {defaultChoiceDisplayText=Continue}}]
 }
 Speech_C_46 {
  id=#22
  class=Speech_C
  outputs []
  properties {generalConfig {defaultMessage=You go on ahead. I'll start on my own time.}}
  sub_nodes [ConversationChoiceNode_66 {
    class=ConversationChoiceNode
    properties {defaultChoiceDisplayText=Continue, choiceTags {gameplayTags [{tagName=Conversation.Choice.RageDebris.CharacterStories.01}]}}
   }]
 }
]
